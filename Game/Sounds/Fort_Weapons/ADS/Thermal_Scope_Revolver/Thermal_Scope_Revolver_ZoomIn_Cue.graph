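# Sound cue for the thermal scope revolver's aim-down-sights zoom-in, routed to the weapon foley sound class.
# A single wave player feeds a Modulator left at default settings, with the cue's overall volume reduced.
# Attenuation overrides disable distance falloff and spatialization, so it plays as flat 2D audio.
# A concurrency override blocks new instances at the limit rather than stopping playing ones.
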SoundCue /Game/Sounds/Fort_Weapons/ADS/Thermal_Scope_Revolver/Thermal_Scope_Revolver_ZoomIn_Cue
root=Modulator_2
nodes [
 {id=Modulator_2, children [WavePlayer_0]}
 {id=WavePlayer_0, properties {soundWaveAssetPtr=&/Themal_Scope_Revolver_Zoom_In}}
]
properties {
 bOverrideAttenuation=true
 bOverrideConcurrency=true
 maxDistance=2097152
 soundClassObject=&Fort_WeaponFoley
 volumeMultiplier=0.301
 attenuationOverrides {bAttenuate=false, bSpatialize=false, falloffDistance=1500, nonSpatializedRadiusStart=200, attenuationShapeExtents {x=1200}}
 concurrencyOverrides {resolutionRule=PreventNew}
}
